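# Default Lit, Nanite-enabled material for the MMGR prop among the SparksMaps Patchwork environment props.
# It samples the prop's T_Sparks_MMGR_prop_D texture and doubles its red channel with a single Multiply node.
# The graph shows no connection from that Multiply to a material output, and no parameters.
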Material /SparksMaps/Environment/Props/Patchwork/Materials/M_Sparks_MMGR_prop
graph {
 properties {bUsedWithNanite=true, pixelDepthOffsetMode=PDOM_Legacy, usedShadingModels=Default Lit}
 nodes [Multiply(A=TextureSample() {texture=/SparksMaps/Environment/Props/Patchwork/Textures/T_Sparks_MMGR_prop_D}.R) {constB=2}]
}
