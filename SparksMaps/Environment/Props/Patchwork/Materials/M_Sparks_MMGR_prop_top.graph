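# Nanite-enabled Default Lit material for the top of the Patchwork "MMGR" environment prop.
# A diagonally panning mask texture has its red channel run through a power, tinted cyan and halved, producing an animated colored pattern.
# Separately, the red channel of the D texture is doubled, and a constant warm beige color stands on its own.
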Material /SparksMaps/Environment/Props/Patchwork/Materials/M_Sparks_MMGR_prop_top
graph {
 properties {bUsedWithNanite=true, pixelDepthOffsetMode=PDOM_Legacy, usedShadingModels=Default Lit}
 nodes [
  Constant4Vector() {constant=<rgba 0.542 0.497 0.449 1>}
  Multiply(A=TextureSample() {texture=/SparksMaps/Environment/Props/Patchwork/Textures/T_Sparks_MMGR_prop_D}.R) {constB=2}
  n1 = TextureSample(UVs=Panner() {speedX=-0.03, speedY=0.03}) {samplerType=SAMPLERTYPE_Masks, texture=/SparksMaps/Environment/Props/Patchwork/Textures/T_Sparks_MMGR_prop_S}
  Multiply(A=Multiply(A=Power(Base=@n1.R), B=Constant4Vector() {constant=<rgba 0.1018 0.8735 0.9089 1>}.RGBA)) {constB=0.5}
 ]
}
